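# Niagara system for a Sparks Season 8 background spotlight effect: a fan of colored mesh light beams.
# Three emitters burst beams along a tilted ring arc, oriented by shape vector, with color and X scale picked by partition index.
# Thick and soft fans pulse alpha with the NPC audio BeatLifetime parameter; thin beams fade over looping lifetimes.
# Soft and thin beams drift through slow random rotation, layered by camera offsets.
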NiagaraSystem /SparksMaps/Effects/Environment/S8/BGSpotlights/NS_Sparks_BG_Spotlight_Env_S8
dependencies {
 usedDataInterfaces [&NiagaraDataInterfaceArrayColor, &NiagaraDataInterfaceArrayFloat, &NiagaraDataInterfaceCurve]
 usedDynamicInputs []
 usedModules [&SystemState]
}
emitters {
 MeshThickFan {
  emitterUpdateScript [
   &EmitterState(
    Inactive Response=::NewEnumerator1
    Life Cycle Mode=::NewEnumerator1
    Loop Behavior=::NewEnumerator1
    Loop Duration Mode=::NewEnumerator1
    UseLoopDelay=false
   )
   &SpawnBurst_Instantaneous(Spawn Count=10, Use Spawn Probability=true, UseLoopCountLimit=false)
   &SpawnRate(SpawnRate=0, Use Spawn Probability=false) {enabled=false}
  ]
  particleSpawnScript [
   &PartitionParticles(Number Of Partitions=10, Particles Per Partition=1, Partition Mode:ENiagara_PartitionMode=::NewEnumerator1)
   &InitializeParticle(
    Color=&SelectLinearColorFromArray(
     Array Sampling Mode=::NewEnumerator1
     Direct Array Index=$Particles.PartitionParticles.Partition
     Color Selection Array=DataInterface() {propertyValues {ColorData [
        <rgba 0 4.5213 5 1>
        <rgba 5 4.2797 2.474 1>
        <rgba 5 2.4386 4.829 1>
        <rgba 0 4.5213 5 1>
        <rgba 5 2.4386 4.829 1>
        <rgba 5 4.2797 2.474 1>
        <rgba 0 4.5213 5 1>
        <rgba 5 4.2797 2.474 1>
        <rgba 2.6562 1.2066 5 1>
        <rgba 0 4.5213 5 1>
       ]}}
     Random Seed=?
    )
    Lifetime Max=8
    Lifetime Min=5
    Lifetime Mode=::NewEnumerator1
    Mesh Scale=&MakeVector(
     X=&Multiply_Float(
      A=1.15
      B=&SelectFloatFromArray(
       Direct Array Index=$Particles.PartitionParticles.Partition
       Float Selection Array=DataInterface() {propertyValues {FloatData [7,6.5,5.5,5,5,5,5,5.5,6.5,7]}}
      )
     )
     Y=1
     Z=1
    )
    Mesh Scale Mode=::NewEnumerator4
    UsePositionOffset=false
    Write Material Random=true
   )
   &ShapeLocation(
    Apply To Particle Position=::NewEnumerator1
    MaskSpawnGroup=false
    Override Seed=false
    Ring / Disc Distribution Mode:ENiagaraSphereDistributionMode=::NewEnumerator1
    Ring / Disc Mode:ENiagaraRingDiscMode=::NewEnumerator1
    Ring Radius=1
    Rotation Angle=-57
    Rotation Angle Type:ENiagara_AngleInput=::NewEnumerator0
    Rotation Axis=<-1 0 0>
    Rotation Coordinate Space=::Simulation
    Rotation Mode=::NewEnumerator1
    Shape Primitive=::NewEnumerator4
    U Position=&ReturnNormalizedExecIndex(Normalized Index Scale=0.475)
    UseOwnerScale=true
   )
   &InitialMeshOrientation(Mesh Orientation Mode=::NewEnumerator2, Orientation Vector=<0 0 1>, Use Rotation Vector=false) {enabled=false}
   &InitialMeshOrientation(Mesh Orientation Mode=::NewEnumerator2, Orientation Vector=$StackContext.ShapeLocation.ShapeVector, Use Rotation Vector=false) {module=InitialMeshOrientation001}
  ]
  particleUpdateScript [
   &ParticleState(
    Kill Particles When Lifetime Has Elapsed=false
    Let Infinitely Lived Particles Die When Emitter Deactivates=true
    Loop Particles Lifetime=true
   )
   &ScaleColor(
    Scale Alpha=&FloatFromCurve(
     FloatCurve=DataInterface() {propertyValues {Curve {keys [
         {interpMode=RCIM_Cubic, tangentMode=RCTM_User, time=0, value=0.85}
         {interpMode=RCIM_Cubic, tangentMode=RCTM_User, time=0.5, value=1}
         {tangentMode=RCTM_User, time=1, value=0.85}
        ]}}}
     Scale Curve=&RandomRangeFloat(Minimum=0.9, Maximum=1)
    )
    Scale RGB=<2.5 2.5 2.5>
    ScaleA=true
    ScaleRGB=true
   )
   &ScaleColor(
    Scale Alpha=&FloatFromCurve(
     FloatCurve=DataInterface() {propertyValues {Curve {keys [
         {arriveTangent=9.0725, interpMode=RCIM_Cubic, time=0, value=0.95}
         <at 0.1 1.1 RCIM_Cubic>
         {arriveTangent=-0.0088, interpMode=RCIM_Cubic, leaveTangent=-0.0088, tangentMode=RCTM_User, time=1, value=0.95}
        ]}}}
     CurveIndex=$NPC.NPC_Sparks_Audio_Parameters001.BeatLifetime
    )
    ScaleA=true
    ScaleRGB=false
   ) {module=ScaleColor001}
  ]
  rendererClasses [&NiagaraMeshRendererProperties]
  renderers [{rendererClass=&NiagaraMeshRendererProperties, rendererIndex=0}]
 }
 MeshThickFanSoft {
  emitterUpdateScript [
   &EmitterState(
    Inactive Response=::NewEnumerator1
    Life Cycle Mode=::NewEnumerator1
    Loop Behavior=::NewEnumerator1
    Loop Duration Mode=::NewEnumerator1
    UseLoopDelay=false
   )
   &SpawnBurst_Instantaneous(Spawn Count=20, Use Spawn Probability=true, UseLoopCountLimit=false)
   &SpawnRate(SpawnRate=0, Use Spawn Probability=false) {enabled=false}
  ]
  particleSpawnScript [
   &PartitionParticles(Number Of Partitions=10, Particles Per Partition=1, Partition Mode:ENiagara_PartitionMode=::NewEnumerator1)
   &InitializeParticle(
    Color=&SelectLinearColorFromArray(
     Array Sampling Mode=::NewEnumerator1
     Direct Array Index=$Particles.PartitionParticles.Partition
     Color Selection Array=DataInterface() {propertyValues {ColorData [
        <rgba 0 4.5213 5 1>
        <rgba 5 4.2797 2.474 1>
        <rgba 5 2.4386 4.829 1>
        <rgba 0 4.5213 5 1>
        <rgba 5 2.4386 4.829 1>
        <rgba 5 4.2797 2.474 1>
        <rgba 0 4.5213 5 1>
        <rgba 5 4.2797 2.474 1>
        <rgba 2.6562 1.2066 5 1>
        <rgba 0 4.5213 5 1>
       ]}}
     Random Seed=?
    )
    Lifetime Max=3
    Lifetime Mode=::NewEnumerator1
    Mesh Scale=&MakeVector(
     X=&Multiply_Float(
      A=0.53
      B=&SelectFloatFromArray(
       Direct Array Index=$Particles.PartitionParticles.Partition
       Float Selection Array=DataInterface() {propertyValues {FloatData [5,5,5,5,5,5,5,5,5,5]}}
      )
     )
     Y=1
     Z=1
    )
    Mesh Scale Mode=::NewEnumerator4
    UsePositionOffset=false
    Write Material Random=true
   )
   &ShapeLocation(
    Apply To Particle Position=::NewEnumerator1
    MaskSpawnGroup=false
    Override Seed=false
    Ring / Disc Distribution Mode:ENiagaraSphereDistributionMode=::NewEnumerator1
    Ring / Disc Mode:ENiagaraRingDiscMode=::NewEnumerator1
    Ring Radius=1
    Rotation Angle=-57
    Rotation Angle Type:ENiagara_AngleInput=::NewEnumerator0
    Rotation Axis=<-1 0 0>
    Rotation Coordinate Space=::Simulation
    Rotation Mode=::NewEnumerator1
    Shape Primitive=::NewEnumerator4
    U Position=&ReturnNormalizedExecIndex(Normalized Index Scale=0.475)
    UseOwnerScale=true
   )
   &InitialMeshOrientation(Mesh Orientation Mode=::NewEnumerator2, Orientation Vector=<0 0 1>, Use Rotation Vector=false) {enabled=false}
   &InitialMeshOrientation(Mesh Orientation Mode=::NewEnumerator2, Orientation Vector=$StackContext.ShapeLocation.ShapeVector, Use Rotation Vector=false) {module=InitialMeshOrientation001}
  ]
  particleUpdateScript [
   &ParticleState(
    Kill Particles When Lifetime Has Elapsed=false
    Let Infinitely Lived Particles Die When Emitter Deactivates=true
    Loop Particles Lifetime=true
   )
   &ScaleColor(
    Scale Alpha=&FloatFromCurve(
     FloatCurve=DataInterface() {propertyValues {Curve {keys [
         {interpMode=RCIM_Cubic, tangentMode=RCTM_User, time=0, value=0.85}
         {interpMode=RCIM_Cubic, tangentMode=RCTM_User, time=0.5, value=1}
         {tangentMode=RCTM_User, time=1, value=0.85}
        ]}}}
     Scale Curve=&RandomRangeFloat(Minimum=2, Maximum=3)
    )
    ScaleA=true
    ScaleRGB=false
   )
   &ScaleColor(Scale Alpha=0.97, Scale RGB=<0.2 0.2 0.2>, ScaleA=true, ScaleRGB=true) {enabled=false, module=ScaleColor001}
   &ScaleColor(
    Scale Alpha=&FloatFromCurve(
     FloatCurve=DataInterface() {propertyValues {Curve {keys [
         {arriveTangent=9.0725, interpMode=RCIM_Cubic, time=0, value=0.95}
         <at 0.1 1.1 RCIM_Cubic>
         {arriveTangent=-0.0088, interpMode=RCIM_Cubic, leaveTangent=-0.0088, tangentMode=RCTM_User, time=1, value=0.95}
        ]}}}
     CurveIndex=$NPC.NPC_Sparks_Audio_Parameters001.BeatLifetime
    )
    ScaleA=true
    ScaleRGB=false
   ) {module=ScaleColor002}
   &UpdateMeshOrientation(
    Rotation Rate=&SelectFloatFromArray(
     Array Sampling Mode=::NewEnumerator0
     Direct Array Index=?
     Float Selection Array=DataInterface() {propertyValues {FloatData [-0.01,0.01]}}
     Random Seed=0
    )
    Rotation Vector=&RandomRangeVector(Minimum=<0.5 -0.1 0>, Maximum=<1 0.1 0>)
   )
   &CameraOffset(Camera Offset Amount=-5)
  ]
  rendererClasses [&NiagaraMeshRendererProperties]
  renderers [{rendererClass=&NiagaraMeshRendererProperties, rendererIndex=0}]
 }
 MeshThinWandering {
  emitterUpdateScript [
   &EmitterState(
    Inactive Response=::NewEnumerator1
    Life Cycle Mode=::NewEnumerator1
    Loop Behavior=::NewEnumerator1
    Loop Duration Mode=::NewEnumerator1
    UseLoopDelay=false
   )
   &SpawnBurst_Instantaneous(Spawn Count=10, Use Spawn Probability=true, UseLoopCountLimit=false)
   &SpawnBurst_Instantaneous(Spawn Count=10, Use Spawn Probability=true, UseLoopCountLimit=false) {module=SpawnBurst_Instantaneous001}
   &SpawnBurst_Instantaneous(Spawn Count=10, Use Spawn Probability=true, UseLoopCountLimit=false) {module=SpawnBurst_Instantaneous002}
  ]
  particleSpawnScript [
   &PartitionParticles(Number Of Partitions=10, Particles Per Partition=1, Partition Mode:ENiagara_PartitionMode=::NewEnumerator1)
   &InitializeParticle(
    Color=&SelectLinearColorFromArray(
     Array Sampling Mode=::NewEnumerator1
     Direct Array Index=$Particles.PartitionParticles.Partition
     Color Selection Array=DataInterface() {propertyValues {ColorData [
        <rgba 0 4.5213 5 1>
        <rgba 5 4.2797 2.474 1>
        <rgba 5 2.4386 4.829 1>
        <rgba 0 4.5213 5 1>
        <rgba 5 2.4386 4.829 1>
        <rgba 5 4.2797 2.474 1>
        <rgba 0 4.5213 5 1>
        <rgba 5 4.2797 2.474 1>
        <rgba 2.6562 1.2066 5 1>
        <rgba 0 4.5213 5 1>
       ]}}
     Random Seed=?
    )
    Lifetime Max=8
    Lifetime Min=5
    Lifetime Mode=::NewEnumerator1
    Mesh Scale=&MakeVector(
     X=&Multiply_Float(
      A=&RandomRangeFloat(Minimum=0.95, Maximum=1.1)
      B=&SelectFloatFromArray(
       Direct Array Index=$Particles.PartitionParticles.Partition
       Float Selection Array=DataInterface() {propertyValues {FloatData [7,6.5,5.5,5,5,5,5,5.5,6.5,7]}}
      )
     )
     Y=1
     Z=1
    )
    Mesh Scale Mode=::NewEnumerator4
    UsePositionOffset=false
    Write Material Random=true
   )
   &ShapeLocation(
    Apply To Particle Position=::NewEnumerator1
    MaskSpawnGroup=false
    Override Seed=false
    Ring / Disc Distribution Mode:ENiagaraSphereDistributionMode=::NewEnumerator1
    Ring / Disc Mode:ENiagaraRingDiscMode=::NewEnumerator1
    Ring Radius=1
    Rotation Angle=-57
    Rotation Angle Type:ENiagara_AngleInput=::NewEnumerator0
    Rotation Axis=<-1 0 0>
    Rotation Coordinate Space=::Simulation
    Rotation Mode=::NewEnumerator1
    Shape Primitive=::NewEnumerator4
    U Position=&ReturnNormalizedExecIndex(Normalized Index Scale=0.475)
    UseOwnerScale=true
   )
   &InitialMeshOrientation(Mesh Orientation Mode=::NewEnumerator2, Orientation Vector=<0 0 1>, Use Rotation Vector=false) {enabled=false}
   &InitialMeshOrientation(Mesh Orientation Mode=::NewEnumerator2, Orientation Vector=$StackContext.ShapeLocation.ShapeVector, Use Rotation Vector=false) {enabled=false, module=InitialMeshOrientation001}
   &InitialMeshOrientation(Mesh Orientation Mode=::NewEnumerator2, Orientation Vector=$StackContext.ShapeLocation.ShapeVector, Use Rotation Vector=false) {module=InitialMeshOrientation002}
  ]
  particleUpdateScript [
   &ParticleState(
    Kill Particles When Lifetime Has Elapsed=false
    Let Infinitely Lived Particles Die When Emitter Deactivates=true
    Loop Particles Lifetime=true
   )
   &UpdateMeshOrientation(
    Rotation Rate=&SelectFloatFromArray(
     Array Sampling Mode=::NewEnumerator0
     Direct Array Index=?
     Float Selection Array=DataInterface() {propertyValues {FloatData [-0.03,0.03]}}
     Random Seed=0
    )
    Rotation Vector=&RandomRangeVector(Minimum=<0.5 -0.1 0>, Maximum=<1 0.1 0>)
   )
   &ScaleColor(
    Scale Alpha=&FloatFromCurve(
     FloatCurve=DataInterface() {propertyValues {Curve {keys [
         {arriveTangent=9.0725, interpMode=RCIM_Cubic, leaveTangent=9.0725, tangentMode=RCTM_User, time=0, value=0.25}
         {arriveTangent=0.2168, interpMode=RCIM_Cubic, leaveTangent=0.2168, tangentMode=RCTM_User, time=0.1408, value=0.9847}
         {arriveTangent=-0.7525, interpMode=RCIM_Cubic, leaveTangent=-0.7525, tangentMode=RCTM_User, time=1, value=0.25}
        ]}}}
     Scale Curve=&RandomRangeFloat(Minimum=0.75, Maximum=1)
    )
    Scale RGB=<6 6 6>
    ScaleA=true
    ScaleRGB=true
   )
   &CameraOffset(Camera Offset Amount=10)
  ]
  rendererClasses [&NiagaraMeshRendererProperties]
  renderers [{rendererClass=&NiagaraMeshRendererProperties, rendererIndex=0}]
 }
}
summary {
 systemName=NS_Sparks_BG_Spotlight_Env_S8
 emitters [
  {emitterName=MeshThickFan, rendererClasses [&NiagaraMeshRendererProperties]}
  {emitterName=MeshThinWandering, rendererClasses [&NiagaraMeshRendererProperties]}
  {emitterName=MeshThickFanSoft, rendererClasses [&NiagaraMeshRendererProperties]}
 ]
 userVariables []
}
